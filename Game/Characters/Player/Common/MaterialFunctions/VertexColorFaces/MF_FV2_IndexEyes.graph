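# Player face material function that shows indexed eye and brow variants by collapsing non-matching vertices to the object position through World Position Offset.
# Left and right index parameters, selected by pre-skinned local X, are divided by ten and remapped through curve atlas rows.
# Sphere masks match the results against vertex color green for eyes and red for brows; pure white vertices always stay visible.
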MaterialFunction /Game/Characters/Player/Common/MaterialFunctions/VertexColorFaces/MF_FV2_IndexEyes
graph {nodes [
  n1 = GetMaterialAttributes(None=FunctionInput() {inputName=Material, inputType=FunctionInput_MaterialAttributes}) {attributeGetTypes [#1]}
  n2 = NamedRerouteDeclaration(None=@n1.WorldPositionOffset) {name=WPO_IN, nodeColor=<rgba 0.5 1 0 1>}
  n3 = Constant3Vector() {constant=<rgba 1 1 1 0>}
  n4 = VertexColor()
  n5 = Constant()
  n6 = ScalarParameter() {group=Flipbook Eyes, parameterName=flipbook_face_L_eye_index}
  n7 = ScalarParameter() {group=Flipbook Eyes, parameterName=flipbook_face_R_eye_index}
  n8 = NamedRerouteUsage()
  n9 = CurveAtlasRowParameter(CurveTime=Divide(A=LinearInterpolate(A=@n6, B=@n7, Alpha=@n8)) {constB=10}) {
   atlas=&/CA_FlipbookV2_Remapping
   curve=&/C_EyesAndMouthDefault
   group=Flipbook Eyes
   parameterName=EyeIndexRemapCurve
  }
  n10 = VertexColor()
  n11 = ScalarParameter() {group=Flipbook Eyes, parameterName=Flipbook_face_L_brow_index}
  n12 = ScalarParameter() {group=Flipbook Eyes, parameterName=Flipbook_face_R_brow_index}
  n13 = NamedRerouteUsage()
  n14 = CurveAtlasRowParameter(CurveTime=Divide(A=LinearInterpolate(A=@n11, B=@n12, Alpha=@n13)) {constB=10}) {
   atlas=&CA_FlipbookV2_Remapping
   curve=&C_EyesAndMouthDefault
   group=Flipbook Eyes
   parameterName=BrowIndexRemapCurve
  }
  n15 = VertexColor()
  n16 = Subtract(
   A=If(A=DotProduct(A=@n3, B=@n4), B=Constant() {r=3}, "A > B"=@n5, "A == B"=@n5, "A < B"=Constant() {r=1})
   B=Add(A=SphereMask(A=@n9.R, B=@n10.G) {attenuationRadius=0.02}, B=SphereMask(A=@n14.R, B=@n15.R) {attenuationRadius=0.02})
  )
  n17 = Multiply(A=NamedRerouteUsage(), B=Saturate(None=Round(None=@n16)))
  FunctionOutput(None=SetMaterialAttributes!(MaterialAttributes=@n1.MaterialAttributes, WorldPositionOffset=Add(A=@n2, B=@n17)) {attributeSetTypes [#1]})
  Min(A=Min(A=@n4.R, B=@n4.R), B=@n4.B)
  n18 = Round(None=MaterialFunctionCall(Float3=LocalPosition() {localOrigin=InstancePreSkinning}.XYZ) {materialFunction=&BreakOutFloat3Components}.R)
  NamedRerouteDeclaration(None=OneMinus(None=Saturate(None=@n18))) {name=Left0/Right1, nodeColor=<rgba 1 0 0.1172 1>}
  n19 = Add(A=WorldPosition().XYZ, B=NamedRerouteUsage())
  n20 = ObjectPositionWS()
  NamedRerouteDeclaration(None=Multiply(A=Subtract(A=@n19, B=@n20)) {constB=-1}) {name=Vertex Color Hide, nodeColor=<rgba 0.6875 0 1 1>}
 ]}
